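# Niagara system for the IceMagic emote that bursts eight icon sprites outward from a central point to fixed offsets around the character, with an age-driven overshoot curve.
# A random per-emitter layout assigns each partitioned particle a sub-image, which sets its size range; color follows time of day.
# User.LocationCurve and User.RotationCurve externally drive vertical drift and icon rotation.
# A disabled duplicate emitter, Icons001, uses sine waves for that motion instead.
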NiagaraSystem /BRCosmetics/Animation/Game/MainPlayer/Emotes/IceMagic/FX/NS_Emote_IceMagic_Icons
dependencies {
 usedDataInterfaces [&NiagaraDataInterfaceArrayFloat, &NiagaraDataInterfaceCurve]
 usedDynamicInputs []
 usedModules [&SystemState]
}
emitters {
 Icons {
  emitterSpawnScript [/SetVariables(Emitter.Random:NiagaraInt32=&RandomRangeInt(Minimum=0, Maximum=2)) {isSetParameters=true, module=SetVariables_#1}]
  emitterUpdateScript [
   &EmitterState(
    Inactive Response=::NewEnumerator1
    Life Cycle Mode=::NewEnumerator1
    Loop Behavior=::NewEnumerator1
    Loop Duration Mode=::NewEnumerator1
    UseLoopDelay=false
   )
   &SpawnBurst_Instantaneous(Spawn Count=8, Use Spawn Probability=false, UseLoopCountLimit=false)
  ]
  particleSpawnScript [
   &PartitionParticles(Number Of Partitions=8, Particles Per Partition=1, Partition Mode:ENiagara_PartitionMode_v2=::NewEnumerator1)
   /SetVariables(Particles.SubImageIndex=&MergeHelper_FloatSwitch(
     0=&MergeHelper_FloatSwitch(0=0, 1=2, 2=0, 3=0, 4=0, 5=0, 6=1, 7=0, Index=$Particles.PartitionParticles.Partition, Num=::NewEnumerator6)
     1=&MergeHelper_FloatSwitch(0=0, 1=0, 2=1, 3=0, 4=0, 5=2, 6=0, 7=0, Index=$Particles.PartitionParticles.Partition, Num=::NewEnumerator6)
     2=&MergeHelper_FloatSwitch(0=0, 1=0, 2=2, 3=0, 4=1, 5=0, 6=0, 7=0, Index=$Particles.PartitionParticles.Partition, Num=::NewEnumerator6)
     3=?
     Index=$Emitter.Random
     Num=::NewEnumerator1
    )) {isSetParameters=true, module=SetVariables_#2}
   &InitializeParticle(
    Color=&FNTimeOfDayColor(Night Color=<rgba 0.4338 0.6661 1 1>, Morning Color=<rgba 1 1 1 1>, Day Color=<rgba 1 1 1 1>, Evening Color=<rgba 1 1 1 1>)
    Sprite Rotation Mode=::NewEnumerator1
    Sprite Size Mode=::NewEnumerator3
    Uniform Sprite Size=&Multiply_Float(
     A=&MergeHelper_FloatSwitch(
      0=&RandomRangeFloat(Minimum=10, Maximum=30)
      1=&RandomRangeFloat(Minimum=30, Maximum=40)
      2=&RandomRangeFloat(Minimum=35, Maximum=45)
      3=?
      Index=&MakeIntFromFloat(Float=$Particles.SubImageIndex)
      Num=::NewEnumerator1
     )
     B=1
    )
    UsePositionOffset=false
    Write Material Random=true
   )
   &CameraOffset(Camera Offset Amount=10)
  ]
  particleUpdateScript [
   &ParticleState(Kill Particles When Lifetime Has Elapsed=false, Let Infinitely Lived Particles Die When Emitter Deactivates=true)
   /SetVariables(Particles.IconRotation:NiagaraFloat=&Multiply_Float(
     A=&Lerp_Float(A=-8, B=8, Alpha=$User.RotationCurve)
     B=&SelectFloatFromArray(
      Array Sampling Mode=::NewEnumerator0
      Direct Array Index=?
      Float Selection Array=DataInterface() {propertyValues {FloatData [-1,1]}}
      Random Seed=0
     )
    )) {isSetParameters=true, module=SetVariables_#3}
   &SystemLocation(
    MaskSpawnGroup=false
    Offset=&Add_Vector(
     A=&MakeVector(
      X=0
      Y=0
      Z=&Multiply_Float(A=$User.LocationCurve, B=&RandomRangeFloat(Minimum=1.5, Maximum=10))
     )
     B=&Lerp_Vector(
      A=<0 0 100>
      B=&MergeHelper_VectorSwitch(
       0=<-30 30 180>
       1=<-60 -10 130>
       2=<-65 20 90>
       3=<-60 -20 50>
       4=<50 -30 180>
       5=<60 20 155>
       6=<70 20 90>
       7=<50 40 60>
       Index=$Particles.PartitionParticles.Partition
       Num=::NewEnumerator6
      )
      Alpha=&FloatFromCurve(
       FloatCurve=DataInterface() {propertyValues {Curve {
          defaultValue=1
          keys [
           {arriveTangent=4.1785, interpMode=RCIM_Cubic, leaveTangent=4.1785, tangentMode=RCTM_User, time=0, value=0.5503}
           {arriveTangent=2.4356, interpMode=RCIM_Cubic, leaveTangent=2.4356, tangentMode=RCTM_User, time=0.0823, value=0.892}
           {arriveTangent=0.0224, interpMode=RCIM_Cubic, leaveTangent=0.0224, tangentMode=RCTM_User, time=0.1879, value=1.05}
           {interpMode=RCIM_Cubic, leaveTangent=3.8894, time=0.2176, value=1}
          ]
         }}}
       CurveIndex=$Emitter.Age
      )
     )
    )
   ) {module=SystemLocation001}
  ]
  rendererClasses [&NiagaraSpriteRendererProperties]
  renderers [{rendererClass=&NiagaraSpriteRendererProperties}]
 }
 Icons001 {
  bEnabled=false
  emitterSpawnScript [/SetVariables(Emitter.Random:NiagaraInt32=&RandomRangeInt(Minimum=0, Maximum=2)) {isSetParameters=true, module=SetVariables_#1}]
  emitterUpdateScript [
   &EmitterState(
    Inactive Response=::NewEnumerator1
    Life Cycle Mode=::NewEnumerator1
    Loop Behavior=::NewEnumerator1
    Loop Duration Mode=::NewEnumerator1
    UseLoopDelay=false
   )
   &SpawnBurst_Instantaneous(Spawn Count=8, Use Spawn Probability=false, UseLoopCountLimit=false)
  ]
  particleSpawnScript [
   &PartitionParticles(Number Of Partitions=8, Particles Per Partition=1, Partition Mode:ENiagara_PartitionMode_v2=::NewEnumerator1)
   /SetVariables(Particles.SubImageIndex=&MergeHelper_FloatSwitch(
     0=&MergeHelper_FloatSwitch(0=0, 1=2, 2=0, 3=0, 4=0, 5=0, 6=1, 7=0, Index=$Particles.PartitionParticles.Partition, Num=::NewEnumerator6)
     1=&MergeHelper_FloatSwitch(0=0, 1=0, 2=1, 3=0, 4=0, 5=2, 6=0, 7=0, Index=$Particles.PartitionParticles.Partition, Num=::NewEnumerator6)
     2=&MergeHelper_FloatSwitch(0=0, 1=0, 2=2, 3=0, 4=1, 5=0, 6=0, 7=0, Index=$Particles.PartitionParticles.Partition, Num=::NewEnumerator6)
     3=?
     Index=$Emitter.Random
     Num=::NewEnumerator1
    )) {isSetParameters=true, module=SetVariables_#2}
   &InitializeParticle(
    Color=&FNTimeOfDayColor(Night Color=<rgba 0.4338 0.6661 1 1>, Morning Color=<rgba 1 1 1 1>, Day Color=<rgba 1 1 1 1>, Evening Color=<rgba 1 1 1 1>)
    Sprite Rotation Mode=::NewEnumerator1
    Sprite Size Mode=::NewEnumerator3
    Uniform Sprite Size=&Multiply_Float(
     A=&MergeHelper_FloatSwitch(
      0=&RandomRangeFloat(Minimum=10, Maximum=30)
      1=&RandomRangeFloat(Minimum=30, Maximum=40)
      2=&RandomRangeFloat(Minimum=35, Maximum=45)
      3=?
      Index=&MakeIntFromFloat(Float=$Particles.SubImageIndex)
      Num=::NewEnumerator1
     )
     B=1
    )
    UsePositionOffset=false
    Write Material Random=true
   )
   &CameraOffset(Camera Offset Amount=10)
  ]
  particleUpdateScript [
   &ParticleState(Kill Particles When Lifetime Has Elapsed=false, Let Infinitely Lived Particles Die When Emitter Deactivates=true)
   &SystemLocation(
    MaskSpawnGroup=false
    Offset=&Add_Vector(
     A=&MakeVector(X=0, Y=0, Z=&Sine(Normalized Angle=$Emitter.Age, Period=0.5, Scale=1, Bias=0))
     B=&Lerp_Vector(
      A=<0 0 100>
      B=&MergeHelper_VectorSwitch(
       0=<-30 30 180>
       1=<-60 -10 130>
       2=<-65 20 90>
       3=<-60 -20 50>
       4=<50 -30 180>
       5=<60 20 155>
       6=<70 20 90>
       7=<50 40 60>
       Index=$Particles.PartitionParticles.Partition
       Num=::NewEnumerator6
      )
      Alpha=&FloatFromCurve(
       FloatCurve=DataInterface() {propertyValues {Curve {
          defaultValue=1
          keys [
           {arriveTangent=4.1785, interpMode=RCIM_Cubic, leaveTangent=4.1785, tangentMode=RCTM_User, time=0, value=0.5503}
           {arriveTangent=2.4356, interpMode=RCIM_Cubic, leaveTangent=2.4356, tangentMode=RCTM_User, time=0.0823, value=0.892}
           {arriveTangent=0.0224, interpMode=RCIM_Cubic, leaveTangent=0.0224, tangentMode=RCTM_User, time=0.1879, value=1.05}
           {interpMode=RCIM_Cubic, leaveTangent=3.8894, time=0.2176, value=1}
          ]
         }}}
       CurveIndex=$Emitter.Age
      )
     )
    )
   )
   /SetVariables(Particles.IconRotation:NiagaraFloat=&Sine(
     Normalized Angle=&Add_Float(
      A=$Emitter.Age
      B=&MergeHelper_FloatSwitch(0=0.5, 1=0, 2=0.5, 3=0, 4=0.5, 5=0, 6=0.5, 7=0, Index=$Particles.PartitionParticles.Partition, Num=::NewEnumerator6)
     )
     Period=1
     Scale=7
     Bias=0.5
    )) {isSetParameters=true, module=SetVariables_#3}
  ]
  rendererClasses [&NiagaraSpriteRendererProperties]
  renderers [{rendererClass=&NiagaraSpriteRendererProperties}]
 }
}
summary {
 systemName=NS_Emote_IceMagic_Icons
 emitters [
  {emitterName=Icons, rendererClasses [&NiagaraSpriteRendererProperties]}
  {bEnabled=false, emitterName=Icons001, rendererClasses [&NiagaraSpriteRendererProperties]}
 ]
 userVariables [
  User.LocationCurve {type=NiagaraFloat, defaultValue=0, description=""}
  User.RotationCurve {type=NiagaraFloat, defaultValue=0, description=""}
 ]
}
